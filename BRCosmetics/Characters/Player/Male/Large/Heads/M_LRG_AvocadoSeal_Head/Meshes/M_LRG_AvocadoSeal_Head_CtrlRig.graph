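# Control Rig for the AvocadoSeal large male head cosmetic that corrects the back-of-neck fur bone, fur_bk_offset_jnt, as the neck moves.
# The bone is first parent-constrained equally to neck_02 and head with averaged interpolation.
# A spherical pose reader then measures neck_01 relative to spine_05, and its output weights an additional translation offset on the fur bone as a pose-driven corrective.
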ControlRigBlueprint /BRCosmetics/Characters/Player/Male/Large/Heads/M_LRG_AvocadoSeal_Head/Meshes/M_LRG_AvocadoSeal_Head_CtrlRig
graphs [
 RigVMFunctionLibrary {links [], nodes []}
 RigVMModel {nodes [
   RigUnit_BeginExecution = &RigUnit_BeginExecution()
   &RigUnit_ParentConstraint(
    AdvancedSettings=(InterpolationType=Average,RotationOrderForFilter=XZY)
    Child=(Type=Bone,Name="fur_bk_offset_jnt")
    Parents=((Item=(Type=Bone,Name="neck_02"),Weight=0.5),(Item=(Type=Bone,Name="head"),Weight=0.5))
    Weight=1
    ExecutePin=@RigUnit_BeginExecution.ExecutePin
   )
   > n1 = &RigUnit_SphericalPoseReader(
    ActiveRegionScaleFactors=(PositiveWidth=0.25,NegativeWidth=1,PositiveHeight=2,NegativeHeight=2)
    ActiveRegionSize=0.1
    Debug=(bDrawDebug=false,DebugScale=25,DebugSegments=20,DebugThickness=0.25)
    DriverAxis=(X=0,Y=-1,Z=0)
    DriverItem=(Type=Bone,Name="neck_01")
    FalloffRegionScaleFactors=(PositiveWidth=0.5,NegativeWidth=0.5,PositiveHeight=2,NegativeHeight=2)
    FalloffSize=0.3
    FlipWidthScaling=false
    OptionalParentItem=(Type=Bone,Name="spine_05")
    RotationOffset=(X=-90,Y=0,Z=-30)
   )
   > &RigUnit_ModifyTransforms(
    ItemToModify=((Item=(Type=Bone,Name="fur_bk_offset_jnt"),Transform=(Rotation=(X=0,Y=0,Z=0,W=1),Translation=(X=3,Y=0,Z=-3),Scale3D=(X=1,Y=1,Z=1))))
    ItemToModify.0.Transform.Rotation=&RigVMFunction_MathQuaternionFromEuler(Euler=(X=0,Y=0,Z=0)).Result
    Weight=@n1.OutputParam
   )
  ]}
]
hierarchy [
 root {loc=<0 0 0>, rot=<0 0 0 1>}
 pelvis {parent=root, loc=<0 -0.7321 92.4218>, rot=<0.0127 0.707 -0.0127 -0.707>}
 spine_01 {parent=pelvis, loc=<5.6002 -0.5128 0>, rot=<0 0 0.034 -0.9994>}
 spine_02 {parent=spine_01, loc=<8.0756 0.0139 0>, rot=<0 0 0.0238 -0.9997>}
 spine_03 {parent=spine_02, loc=<7.0395 0 0>, rot=<0 0 0.0149 -0.9999>}
 spine_04 {parent=spine_03, loc=<8.1023 0 0>, rot=<0 0 -0.0132 -0.9999>}
 spine_05 {parent=spine_04, loc=<10.0141 0 0>, rot=<0 0 -0.0015 -1>}
 clavicle_l {parent=spine_05, loc=<2.5856 -0.2866 -4.3777>, rot=<0.0465 -0.7289 -0.009 -0.6829>}
 neck_01 {parent=spine_05, loc=<10.815 1.9196 0>, rot=<0 0 -0.1864 -0.9825>}
 neck_02 {parent=neck_01, loc=<4.7347 0.3045 0>, rot=<0 0 0.0363 -0.9993>}
 head {parent=neck_02, loc=<5.7632 0.5552 0>, rot=<0 0 0.115 -0.9934>}
 faceAttach {parent=head, loc=<0 0 0>, rot=<0.0029 -0.7071 -0.0029 -0.7071>}
 C_brow_mid {parent=faceAttach, loc=<0 9.9352 5.7601>, rot=<0 0 0 -1>}
 C_jaw {parent=faceAttach, loc=<-0.0134 -0.1283 -1.3998>, rot=<0 0 0 1>}
 C_lip_lower_mid {parent=C_jaw, loc=<0.0134 14.8743 -1.6183>, rot=<0 0 0 -1>}
 L_lip_lower_outer {parent=C_jaw, loc=<2.5175 13.892 -1.7213>, rot=<0.013 0.0445 0.3057 -0.951>}
 R_lip_lower_outer {parent=C_jaw, loc=<-2.4906 13.892 -1.7213>, rot=<0.013 -0.0445 -0.3057 -0.951>}
 teeth_lower {parent=C_jaw, loc=<0.0134 9.2499 -2.4005>, rot=<0 0 0 1>}
 tongue {parent=C_jaw, loc=<0.0141 3.7403 -3.1446>, rot=<-0.1568 0 0 -0.9876>}
 dyn_l_JawFur_01_jnt {parent=C_jaw, loc=<6.814 5.7057 -6.2857>, rot=<0.325 -0.3863 0.1325 -0.853>}
 dyn_l_JawFur_02_jnt {parent=dyn_l_JawFur_01_jnt, loc=<2.4017 0 0>, rot=<0 0 0 1>}
 dyn_l_ChinFur_01_jnt {parent=C_jaw, loc=<5.364 9.312 -9.0683>, rot=<0.1855 -0.6148 -0.2101 -0.7372>}
 dyn_l_ChinFur_02_jnt {parent=dyn_l_ChinFur_01_jnt, loc=<2.4017 0 0>, rot=<0 0 0 1>}
 dyn_ft_l_ChinFur_01_jnt {parent=C_jaw, loc=<3.258 12.2514 -11.3118>, rot=<0.3801 -0.4767 -0.2294 -0.7587>}
 dyn_ft_l_ChinFur_02_jnt {parent=dyn_ft_l_ChinFur_01_jnt, loc=<2.4017 0 0>, rot=<0 0 0 1>}
 dyn_ChinFur_01_jnt {parent=C_jaw, loc=<0.0134 13.5932 -12.4674>, rot=<0.442 -0.442 -0.5519 -0.5519>}
 dyn_ChinFur_02_jnt {parent=dyn_ChinFur_01_jnt, loc=<2.4017 0 0>, rot=<0 0 0 1>}
 dyn_r_JawFur_01_jnt {parent=C_jaw, loc=<-6.7872 5.7057 -6.2854>, rot=<-0.3863 0.325 0.853 -0.1325>}
 dyn_r_JawFur_02_jnt {parent=dyn_r_JawFur_01_jnt, loc=<2.4017 0 0>, rot=<0 0 0 1>}
 dyn_r_ChinFur_01_jnt {parent=C_jaw, loc=<-5.3372 9.312 -9.0684>, rot=<-0.6148 0.1855 0.7372 0.2101>}
 dyn_r_ChinFur_02_jnt {parent=dyn_r_ChinFur_01_jnt, loc=<2.4017 0 0>, rot=<0 0 0 -1>}
 dyn_ft_r_ChinFur_01_jnt {parent=C_jaw, loc=<-3.5937 11.9065 -11.0182>, rot=<-0.4767 0.3801 0.7587 0.2294>}
 dyn_ft_r_ChinFur_02_jnt {parent=dyn_ft_r_ChinFur_01_jnt, loc=<2.4017 0 0>, rot=<0 0 0 1>}
 C_lip_upper_mid {parent=faceAttach, loc=<0.0003 14.7756 -1.3228>, rot=<0.2225 0.0013 -0.0005 -0.9749>}
 C_nose_bridge {parent=faceAttach, loc=<0 9.8539 3.9715>, rot=<-0.2218 0 0 -0.9751>}
 L_brow_mid {parent=faceAttach, loc=<4.4525 8.87 6.1874>, rot=<-0.0001 -0.0042 0.2553 -0.9669>}
 L_brow_outer {parent=faceAttach, loc=<5.3655 7.3586 5.0667>, rot=<-0.0287 -0.04 0.4491 -0.8921>}
 L_cheek_inner {parent=faceAttach, loc=<5.6181 8.3051 1.2081>, rot=<-0.0706 -0.2697 0.3184 -0.906>}
 L_eye {parent=faceAttach, loc=<2.8496 7.6387 4.4979>, rot=<0 0 0.9998 0.0175>}
 L_eye_lid_lower_mid {parent=L_eye, loc=<0 0 0>, rot=<0.0329 0 0 -0.9995>}
 L_eye_lid_upper_mid {parent=L_eye, loc=<0 0 0>, rot=<0.0523 0 0 -0.9986>}
 L_lip_corner {parent=faceAttach, loc=<4.1793 11.482 -3.1713>, rot=<0.003 0.008 -0.0146 -0.9999>}
 L_lip_upper_outer {parent=faceAttach, loc=<2.5466 13.7982 -2.2186>, rot=<0.0359 -0.0728 0.2627 -0.9615>}
 R_brow_mid {parent=faceAttach, loc=<-4.4525 8.87 6.1874>, rot=<-0.0001 0.0042 -0.2553 -0.9669>}
 R_brow_outer {parent=faceAttach, loc=<-5.3655 7.3586 5.0667>, rot=<-0.0287 0.04 -0.4491 -0.8921>}
 R_cheek_inner {parent=faceAttach, loc=<-5.6181 8.3051 1.2081>, rot=<-0.0706 0.2697 -0.3184 -0.906>}
 R_eye {parent=faceAttach, loc=<-2.8496 7.6394 4.4981>, rot=<0 0 0.9998 -0.0175>}
 R_eye_lid_lower_mid {parent=R_eye, loc=<0 0 0>, rot=<0.0329 0 0 -0.9995>}
 R_eye_lid_upper_mid {parent=R_eye, loc=<0 0 0>, rot=<0.0523 0 0 -0.9986>}
 R_lip_corner {parent=faceAttach, loc=<-4.1793 11.482 -3.1713>, rot=<0.003 -0.008 0.0146 -0.9999>}
 R_lip_upper_outer {parent=faceAttach, loc=<-2.5466 13.7982 -2.2186>, rot=<0.0359 0.0728 -0.2627 -0.9615>}
 teeth_upper {parent=faceAttach, loc=<0 9.304 -0.6368>, rot=<0 0 0 1>}
 dyn_l_CornerJawFur_01_jnt {parent=head, loc=<-2.4813 -0.4471 -8.8336>, rot=<0.2385 -0.942 -0.094 -0.2168>}
 dyn_l_CornerJawFur_02_jnt {parent=dyn_l_CornerJawFur_01_jnt, loc=<2.4017 0 0>, rot=<0 0 0 1>}
 dyn_r_CornerJawFur_01_jnt {parent=head, loc=<-2.4816 -0.4471 8.8336>, rot=<0.2168 0.094 0.942 -0.2385>}
 dyn_r_CornerJawFur_02_jnt {parent=dyn_r_CornerJawFur_01_jnt, loc=<2.4017 0 0>, rot=<0 0 0 -1>}
 fur_bk_offset_jnt {parent=neck_02, loc=<5.1705 -10.6502 0>, rot=<0.5384 -0.4584 0.4584 -0.5384>}
 clavicle_r {parent=spine_05, loc=<2.5855 -0.2866 4.3777>, rot=<0.7289 0.0465 -0.6829 0.009>}
]
