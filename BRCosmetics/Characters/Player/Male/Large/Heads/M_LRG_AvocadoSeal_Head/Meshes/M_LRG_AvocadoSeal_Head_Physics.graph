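# Physics asset for the AvocadoSeal large male player head cosmetic, providing bodies and joint limits for its chin and jaw fur.
# Seven two-bone fur chains, each bone with a small offset capsule, hang from C_jaw.
# C_jaw links through faceAttach and head to spine_05 using zero-size capsules and wider limits.
# Fur joints have tight swing and very narrow twist limits, with slightly more swing on center and front chin chains.
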PhysicsAsset /BRCosmetics/Characters/Player/Male/Large/Heads/M_LRG_AvocadoSeal_Head/Meshes/M_LRG_AvocadoSeal_Head_Physics
bodies {
 C_jaw [{shapeType=Capsule, length=0, radius=0}]
 dyn_ChinFur_01_jnt [{shapeType=Capsule, length=2, radius=2, shapeName=dyn_ChinFur_01_jnt_capsule, center=<4 0 0>, rotation=<rot 0 90 90>}]
 dyn_ChinFur_02_jnt [{shapeType=Capsule, length=2, radius=2, shapeName=dyn_ChinFur_02_jnt_capsule, center=<4 0 0>, rotation=<rot 0 90 90>}]
 dyn_ft_l_ChinFur_01_jnt [{shapeType=Capsule, length=2, radius=2, shapeName=dyn_ft_l_ChinFur_01_jnt_capsule, center=<4 0 0>, rotation=<rot 0 90 90>}]
 dyn_ft_l_ChinFur_02_jnt [{shapeType=Capsule, length=2, radius=2, shapeName=dyn_ft_l_ChinFur_02_jnt_capsule, center=<4 0 0>, rotation=<rot 0 90 90>}]
 dyn_ft_r_ChinFur_01_jnt [{shapeType=Capsule, length=2, radius=2, shapeName=dyn_ft_r_ChinFur_01_jnt_capsule, center=<4 0 0>, rotation=<rot 0 90 90>}]
 dyn_ft_r_ChinFur_02_jnt [{shapeType=Capsule, length=2, radius=2, shapeName=dyn_ft_r_ChinFur_02_jnt_capsule, center=<4 0 0>, rotation=<rot 0 90 90>}]
 dyn_l_ChinFur_01_jnt [{shapeType=Capsule, length=2, radius=2, shapeName=dyn_l_ChinFur_01_jnt_capsule, center=<4 0 0>, rotation=<rot 0 90 90>}]
 dyn_l_ChinFur_02_jnt [{shapeType=Capsule, length=2, radius=2, shapeName=dyn_l_ChinFur_02_jnt_capsule, center=<4 0 0>, rotation=<rot 0 90 90>}]
 dyn_l_JawFur_01_jnt [{shapeType=Capsule, length=2, radius=2, shapeName=dyn_l_JawFur_01_jnt_capsule, center=<4 0 0>, rotation=<rot 0 90 90>}]
 dyn_l_JawFur_02_jnt [{shapeType=Capsule, length=2, radius=2, shapeName=dyn_l_JawFur_02_jnt_capsule, center=<4 0 0>, rotation=<rot 0 90 90>}]
 dyn_r_ChinFur_01_jnt [{shapeType=Capsule, length=2, radius=2, shapeName=dyn_r_ChinFur_01_jnt_capsule, center=<4 0 0>, rotation=<rot 0 90 90>}]
 dyn_r_ChinFur_02_jnt [{shapeType=Capsule, length=2, radius=2, shapeName=dyn_r_ChinFur_02_jnt_capsule, center=<4 0 0>, rotation=<rot 0 90 90>}]
 dyn_r_JawFur_01_jnt [{shapeType=Capsule, length=2, radius=2, shapeName=dyn_r_JawFur_01_jnt_capsule, center=<4 0 0>, rotation=<rot 0 90 90>}]
 dyn_r_JawFur_02_jnt [{shapeType=Capsule, length=2, radius=2, shapeName=dyn_r_JawFur_02_jnt_capsule, center=<4 0 0>, rotation=<rot 0 90 90>}]
 faceAttach [{shapeType=Capsule, length=0, radius=0}]
 head [{shapeType=Capsule, length=0, radius=0}]
 spine_05 [{shapeType=Capsule, length=0, radius=0}]
}
constraints [
 {bone1Name=head, bone2Name=spine_05, swing1LimitDegrees=45, swing2LimitDegrees=45, twistLimitDegrees=45}
 {bone1Name=faceAttach, bone2Name=head, swing1LimitDegrees=45, swing2LimitDegrees=45, twistLimitDegrees=45}
 {bone1Name=C_jaw, bone2Name=faceAttach, swing1LimitDegrees=45, swing2LimitDegrees=45, twistLimitDegrees=45}
 {bone1Name=dyn_ChinFur_01_jnt, bone2Name=C_jaw, swing1LimitDegrees=10, swing2LimitDegrees=15, twistLimitDegrees=2}
 {bone1Name=dyn_ChinFur_02_jnt, bone2Name=dyn_ChinFur_01_jnt, swing1LimitDegrees=10, swing2LimitDegrees=15, twistLimitDegrees=2}
 {bone1Name=dyn_ft_l_ChinFur_01_jnt, bone2Name=C_jaw, swing1LimitDegrees=10, swing2LimitDegrees=15, twistLimitDegrees=2}
 {bone1Name=dyn_ft_l_ChinFur_02_jnt, bone2Name=dyn_ft_l_ChinFur_01_jnt, swing1LimitDegrees=10, swing2LimitDegrees=15, twistLimitDegrees=2}
 {bone1Name=dyn_ft_r_ChinFur_01_jnt, bone2Name=C_jaw, swing1LimitDegrees=10, swing2LimitDegrees=15, twistLimitDegrees=2}
 {bone1Name=dyn_ft_r_ChinFur_02_jnt, bone2Name=dyn_ft_r_ChinFur_01_jnt, swing1LimitDegrees=10, swing2LimitDegrees=15, twistLimitDegrees=2}
 {bone1Name=dyn_l_ChinFur_01_jnt, bone2Name=C_jaw, swing1LimitDegrees=10, swing2LimitDegrees=10, twistLimitDegrees=2}
 {bone1Name=dyn_l_ChinFur_02_jnt, bone2Name=dyn_l_ChinFur_01_jnt, swing1LimitDegrees=10, swing2LimitDegrees=10, twistLimitDegrees=2}
 {bone1Name=dyn_l_JawFur_01_jnt, bone2Name=C_jaw, swing1LimitDegrees=10, swing2LimitDegrees=10, twistLimitDegrees=2}
 {bone1Name=dyn_l_JawFur_02_jnt, bone2Name=dyn_l_JawFur_01_jnt, swing1LimitDegrees=10, swing2LimitDegrees=10, twistLimitDegrees=2}
 {bone1Name=dyn_r_ChinFur_01_jnt, bone2Name=C_jaw, swing1LimitDegrees=10, swing2LimitDegrees=10, twistLimitDegrees=2}
 {bone1Name=dyn_r_ChinFur_02_jnt, bone2Name=dyn_r_ChinFur_01_jnt, swing1LimitDegrees=10, swing2LimitDegrees=10, twistLimitDegrees=2}
 {bone1Name=dyn_r_JawFur_01_jnt, bone2Name=C_jaw, swing1LimitDegrees=10, swing2LimitDegrees=10, twistLimitDegrees=2}
 {bone1Name=dyn_r_JawFur_02_jnt, bone2Name=dyn_r_JawFur_01_jnt, swing1LimitDegrees=10, swing2LimitDegrees=10, twistLimitDegrees=2}
]
