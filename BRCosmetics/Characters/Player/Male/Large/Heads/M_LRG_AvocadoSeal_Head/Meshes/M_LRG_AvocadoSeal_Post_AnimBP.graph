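# Post-process AnimBlueprint for the AvocadoSeal large male head, adding secondary physics motion.
# The input pose runs through a RigidBody node using the head's physics asset in C_jaw space, then a head Control Rig.
# Updates compute the RigidBody gravity override via a deprecated helper, and set simSpace and frontendAlpha from DBNO and front-end states.
# A chain of AnimDynamics nodes for jaw and chin fur is unconnected to the output.
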AnimBlueprint /BRCosmetics/Characters/Player/Male/Large/Heads/M_LRG_AvocadoSeal_Head/Meshes/M_LRG_AvocadoSeal_Post_AnimBP
target_skeleton=&/Fortnite_Base_Head_AvocadoSeal_Skeleton
graphs [
 AnimGraph {nodes [
   n1 = Anim Dynamics - Bone: dyn_l_CornerJawFur_01_jnt!(SimulationSpace:AnimPhysSimSpaceType Enum, ComponentPose, Alpha=0.75, ExternalForce="0,0,0") -> (Pose) {
    angularDampingOverride=1
    angularSpringConstant=200
    bAngularSpring=true
    bOverrideAngularDamping=true
    bUseGravityOverride=true
    simulationSpace=RootRelative
    alphaBoolBlend {blendOption=Linear}
    boundBone {boneName=dyn_l_CornerJawFur_01_jnt}
    componentAppliedLinearAccClamp=<6000 6000 6000>
    componentLinearAccScale=<0.5 0.5 0.5>
    componentLinearVelScale=<0.5 0.5 0.5>
    gravityOverride {z=-980}
    physicsBodyDefinitions [{
      boundBone {boneName=dyn_l_CornerJawFur_01_jnt}
      boxExtents=<10 10 10>
      constraintSetup {angularLimitsMax=<2 15 10>, angularLimitsMin=<-2 -15 -10>, angularTarget=<1 0 0>}
      localJointOffset=<8 0 0>
     }]
   }
   n2 = Anim Dynamics - Bone: dyn_l_JawFur_01_jnt!(SimulationSpace:AnimPhysSimSpaceType Enum, ComponentPose=@n1, Alpha=0.75, ExternalForce="0,0,0") -> (Pose) {
    angularDampingOverride=1
    angularSpringConstant=200
    bAngularSpring=true
    bOverrideAngularDamping=true
    bUseGravityOverride=true
    simulationSpace=RootRelative
    alphaBoolBlend {blendOption=Linear}
    boundBone {boneName=dyn_l_JawFur_01_jnt}
    componentAppliedLinearAccClamp=<6000 6000 6000>
    componentLinearAccScale=<0.5 0.5 0.5>
    componentLinearVelScale=<0.5 0.5 0.5>
    gravityOverride {z=-980}
    physicsBodyDefinitions [{
      boundBone {boneName=dyn_l_JawFur_01_jnt}
      boxExtents=<10 10 10>
      constraintSetup {angularLimitsMax=<2 15 10>, angularLimitsMin=<-2 -15 -10>, angularTarget=<1 0 0>}
      localJointOffset=<8 0 0>
     }]
   }
   n3 = Anim Dynamics - Bone: dyn_l_ChinFur_01_jnt!(SimulationSpace:AnimPhysSimSpaceType Enum, ComponentPose=@n2, Alpha=0.75, ExternalForce="0,0,0") -> (Pose) {
    angularDampingOverride=1
    angularSpringConstant=200
    bAngularSpring=true
    bOverrideAngularDamping=true
    bUseGravityOverride=true
    simulationSpace=RootRelative
    alphaBoolBlend {blendOption=Linear}
    boundBone {boneName=dyn_l_ChinFur_01_jnt}
    componentAppliedLinearAccClamp=<6000 6000 6000>
    componentLinearAccScale=<0.5 0.5 0.5>
    componentLinearVelScale=<0.5 0.5 0.5>
    gravityOverride {z=-980}
    physicsBodyDefinitions [{
      boundBone {boneName=dyn_l_ChinFur_01_jnt}
      boxExtents=<10 10 10>
      constraintSetup {angularLimitsMax=<2 15 10>, angularLimitsMin=<-2 -15 -10>, angularTarget=<1 0 0>}
      localJointOffset=<8 0 0>
     }]
   }
   n4 = Anim Dynamics - Bone: dyn_ft_l_ChinFur_01_jnt!(SimulationSpace:AnimPhysSimSpaceType Enum, ComponentPose=@n3, Alpha=0.75, ExternalForce="0,0,0") -> (Pose) {
    angularDampingOverride=1
    angularSpringConstant=200
    bAngularSpring=true
    bOverrideAngularDamping=true
    bUseGravityOverride=true
    simulationSpace=RootRelative
    alphaBoolBlend {blendOption=Linear}
    boundBone {boneName=dyn_ft_l_ChinFur_01_jnt}
    componentAppliedLinearAccClamp=<6000 6000 6000>
    componentLinearAccScale=<0.5 0.5 0.5>
    componentLinearVelScale=<0.5 0.5 0.5>
    gravityOverride {z=-980}
    physicsBodyDefinitions [{
      boundBone {boneName=dyn_ft_l_ChinFur_01_jnt}
      boxExtents=<10 10 10>
      constraintSetup {angularLimitsMax=<2 15 10>, angularLimitsMin=<-2 -15 -10>, angularTarget=<1 0 0>}
      localJointOffset=<8 0 0>
     }]
   }
   n5 = Anim Dynamics - Bone: dyn_ChinFur_01_jnt!(SimulationSpace:AnimPhysSimSpaceType Enum, ComponentPose=@n4, Alpha=0.75, ExternalForce="0,0,0") -> (Pose) {
    angularDampingOverride=1
    angularSpringConstant=200
    bAngularSpring=true
    bOverrideAngularDamping=true
    bUseGravityOverride=true
    simulationSpace=RootRelative
    alphaBoolBlend {blendOption=Linear}
    boundBone {boneName=dyn_ChinFur_01_jnt}
    componentAppliedLinearAccClamp=<6000 6000 6000>
    componentLinearAccScale=<0.5 0.5 0.5>
    componentLinearVelScale=<0.5 0.5 0.5>
    gravityOverride {z=-980}
    physicsBodyDefinitions [{
      boundBone {boneName=dyn_ChinFur_01_jnt}
      boxExtents=<10 10 10>
      constraintSetup {angularLimitsMax=<2 15 10>, angularLimitsMin=<-2 -15 -10>, angularTarget=<1 0 0>}
      localJointOffset=<8 0 0>
     }]
   }
   n6 = Anim Dynamics - Bone: dyn_r_CornerJawFur_01_jnt!(SimulationSpace:AnimPhysSimSpaceType Enum, ComponentPose=@n5, Alpha=0.75, ExternalForce="0,0,0") -> (Pose) {
    angularDampingOverride=1
    angularSpringConstant=200
    bAngularSpring=true
    bOverrideAngularDamping=true
    bUseGravityOverride=true
    simulationSpace=RootRelative
    alphaBoolBlend {blendOption=Linear}
    boundBone {boneName=dyn_r_CornerJawFur_01_jnt}
    componentAppliedLinearAccClamp=<6000 6000 6000>
    componentLinearAccScale=<0.5 0.5 0.5>
    componentLinearVelScale=<0.5 0.5 0.5>
    gravityOverride {z=-980}
    physicsBodyDefinitions [{
      boundBone {boneName=dyn_r_CornerJawFur_01_jnt}
      boxExtents=<10 10 10>
      constraintSetup {angularLimitsMax=<2 15 10>, angularLimitsMin=<-2 -15 -10>, angularTarget=<1 0 0>}
      localJointOffset=<8 0 0>
     }]
   }
   n7 = Anim Dynamics - Bone: dyn_r_JawFur_01_jnt!(SimulationSpace:AnimPhysSimSpaceType Enum, ComponentPose=@n6, Alpha=0.75, ExternalForce="0,0,0") -> (Pose) {
    angularDampingOverride=1
    angularSpringConstant=200
    bAngularSpring=true
    bOverrideAngularDamping=true
    bUseGravityOverride=true
    simulationSpace=RootRelative
    alphaBoolBlend {blendOption=Linear}
    boundBone {boneName=dyn_r_JawFur_01_jnt}
    componentAppliedLinearAccClamp=<6000 6000 6000>
    componentLinearAccScale=<0.5 0.5 0.5>
    componentLinearVelScale=<0.5 0.5 0.5>
    gravityOverride {z=-980}
    physicsBodyDefinitions [{
      boundBone {boneName=dyn_r_JawFur_01_jnt}
      boxExtents=<10 10 10>
      constraintSetup {angularLimitsMax=<2 15 10>, angularLimitsMin=<-2 -15 -10>, angularTarget=<1 0 0>}
      localJointOffset=<8 0 0>
     }]
   }
   n8 = Anim Dynamics - Bone: dyn_r_ChinFur_01_jnt!(SimulationSpace:AnimPhysSimSpaceType Enum, ComponentPose=@n7, Alpha=0.75, ExternalForce="0,0,0") -> (Pose) {
    angularDampingOverride=1
    angularSpringConstant=200
    bAngularSpring=true
    bOverrideAngularDamping=true
    bUseGravityOverride=true
    simulationSpace=RootRelative
    alphaBoolBlend {blendOption=Linear}
    boundBone {boneName=dyn_r_ChinFur_01_jnt}
    componentAppliedLinearAccClamp=<6000 6000 6000>
    componentLinearAccScale=<0.5 0.5 0.5>
    componentLinearVelScale=<0.5 0.5 0.5>
    gravityOverride {z=-980}
    physicsBodyDefinitions [{
      boundBone {boneName=dyn_r_ChinFur_01_jnt}
      boxExtents=<10 10 10>
      constraintSetup {angularLimitsMax=<2 15 10>, angularLimitsMin=<-2 -15 -10>, angularTarget=<1 0 0>}
      localJointOffset=<8 0 0>
     }]
   }
   Anim Dynamics - Bone: dyn_ft_r_ChinFur_01_jnt!(SimulationSpace:AnimPhysSimSpaceType Enum, ComponentPose=@n8, Alpha=0.75, ExternalForce="0,0,0") -> (Pose) {
    angularDampingOverride=1
    angularSpringConstant=200
    bAngularSpring=true
    bOverrideAngularDamping=true
    bUseGravityOverride=true
    simulationSpace=RootRelative
    alphaBoolBlend {blendOption=Linear}
    boundBone {boneName=dyn_ft_r_ChinFur_01_jnt}
    componentAppliedLinearAccClamp=<6000 6000 6000>
    componentLinearAccScale=<0.5 0.5 0.5>
    componentLinearVelScale=<0.5 0.5 0.5>
    gravityOverride {z=-980}
    physicsBodyDefinitions [{
      boundBone {boneName=dyn_ft_r_ChinFur_01_jnt}
      boxExtents=<10 10 10>
      constraintSetup {angularLimitsMax=<2 15 10>, angularLimitsMin=<-2 -15 -10>, angularTarget=<1 0 0>}
      localJointOffset=<8 0 0>
     }]
   }
   n9 = RigidBody!(
    ComponentPose=Local To Component(LocalPose=Input Pose())
    Alpha
    OverrideWorldGravity:Vector=Get Gravity Override()
    ExternalForce
    ComponentLinearAccScale:Vector="0.2,0.2,0.2"
    ComponentLinearVelScale:Vector="0.4,0.3,0.3"
   ) -> (Pose) {
    bClampLinearTranslationLimitToRefPose=true
    bOverrideWorldGravity=true
    overridePhysicsAsset=&/M_LRG_AvocadoSeal_Head_Physics
    simulationSpace=BaseBoneSpace
    alphaBoolBlend {blendOption=Linear}
    alphaScaleBiasClamp {bInterpResult=true, interpSpeedIncreasing=2}
    baseBoneRef {boneName=C_jaw}
    componentAppliedLinearAccClamp=<3000 3000 3000>
    componentLinearAccScale=<0.4 0.3 0.3>
    componentLinearVelScale=<0.4 0.3 0.3>
    simSpaceSettings {
     maxAngularAcceleration=1000
     maxAngularVelocity=3
     maxLinearAcceleration=700
     maxLinearVelocity=500
     velocityScaleZ=0
     worldAlpha=1
     externalLinearDragV=<0.2 0.2>
    }
   }
   n10 = Control Rig(Source=Component To Local(ComponentPose=@n9)) {
    alphaBoolBlend {blendOption=Linear}
    controlRigAssetReference {blueprintRigClass=&/M_LRG_AvocadoSeal_Head_CtrlRig.M_LRG_AvocadoSeal_Head_CtrlRig_C}
   }
   Output Pose(Result=@n10)
  ]}
 EventGraph {nodes [
   Event BlueprintUpdateAnimation()
   > n1 = Sequence!(execute) -> (then_0, then_1, then_2:Exec)
   > Set simSpace(simSpace:AnimPhysSimSpaceType Enum) -> (Output_Get:AnimPhysSimSpaceType Enum)
   > Branch(Condition=OR Boolean(A=Get bIsDBNO(), B=Get bIsInFrontEnd()))
   > n2 = Branch(Condition=OR Boolean(A=Get bIsDBNOCarried(), B=Get bIsInFrontEnd()))
   > Set simSpace(simSpace=World)
   Set simSpace(execute=@n2.else)
   n3 = Gravity Override State Function DEPRECATED(
    execute=@n1.then_2
    CharacterPartAnimInstance=Self-Reference()
    GameTimeInSeconds=GetGameTimeInSeconds()
    GravityOverrideParams=Get Gravity Override Params()
   )
   > Set Gravity Override(Gravity Override=@n3)
   Set frontendAlpha(execute=@n1.then_1, frontendAlpha:double=SelectFloat(A=0.5, B=1, bPickA=Get bIsInFrontEnd())) -> (Output_Get:double)
   TryGetPawnOwner()
  ]}
]
variables [
 frontendAlpha {type=number}
 Gravity Override {type=Vector}
 Gravity Override Params {type=GravityOverrideParamsStruct}
 simSpace {type=AnimPhysSimSpaceType}
]
